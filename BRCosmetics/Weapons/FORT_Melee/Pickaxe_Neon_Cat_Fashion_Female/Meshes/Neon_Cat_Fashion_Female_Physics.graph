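# Physics asset for the Neon Cat Fashion Female pickaxe, rigging its fish as a jointed set of box bodies on the dyn_fish bones.
# A small dyn_fish_MSTR body is the hub of two constraint chains, one running through bones 04 to 07 and the other through 03 down to 01.
# Every joint shares the same narrow swing and twist limits, so the fish can flex only slightly at each segment.
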PhysicsAsset /BRCosmetics/Weapons/FORT_Melee/Pickaxe_Neon_Cat_Fashion_Female/Meshes/Neon_Cat_Fashion_Female_Physics
bodies {
 dyn_fish_01 [{shapeType=Box, extentX=14.4576, extentY=3.7071, extentZ=15.5282, center=<3.5319 0 0>}]
 dyn_fish_02 [{shapeType=Box, extentX=8.982, extentY=4.3015, extentZ=17.2749, center=<1.1349 0 0>}]
 dyn_fish_03 [{shapeType=Box, extentX=9.9319, extentY=4.4747, extentZ=19.5971, center=<1.2583 0 0>}]
 dyn_fish_04 [{shapeType=Box, extentX=8.9181, extentY=4.6455, extentZ=16.872, center=<4.2307 0 0>}]
 dyn_fish_05 [{shapeType=Box, extentX=9.2607, extentY=5.0833, extentZ=14.664, center=<4.3781 0 0>}]
 dyn_fish_06 [{shapeType=Box, extentX=9.0464, extentY=4.1874, extentZ=12.0606, center=<4.2468 0 0>}]
 dyn_fish_07 [{shapeType=Box, extentX=13.4456, extentY=3.4591, extentZ=15.5522, center=<6.4973 0 0>}]
 dyn_fish_MSTR [{shapeType=Box, extentX=1.01, extentY=1.01, extentZ=1.01}]
}
constraints [
 {bone1Name=dyn_fish_04, bone2Name=dyn_fish_MSTR, swing1LimitDegrees=3, swing2LimitDegrees=6, twistLimitDegrees=2}
 {bone1Name=dyn_fish_05, bone2Name=dyn_fish_04, swing1LimitDegrees=3, swing2LimitDegrees=6, twistLimitDegrees=2}
 {bone1Name=dyn_fish_06, bone2Name=dyn_fish_05, swing1LimitDegrees=3, swing2LimitDegrees=6, twistLimitDegrees=2}
 {bone1Name=dyn_fish_07, bone2Name=dyn_fish_06, swing1LimitDegrees=3, swing2LimitDegrees=6, twistLimitDegrees=2}
 {bone1Name=dyn_fish_03, bone2Name=dyn_fish_MSTR, swing1LimitDegrees=3, swing2LimitDegrees=6, twistLimitDegrees=2}
 {bone1Name=dyn_fish_02, bone2Name=dyn_fish_03, swing1LimitDegrees=3, swing2LimitDegrees=6, twistLimitDegrees=2}
 {bone1Name=dyn_fish_01, bone2Name=dyn_fish_02, swing1LimitDegrees=3, swing2LimitDegrees=6, twistLimitDegrees=2}
]
